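# Unlit preview material for inspecting a jump flooding render target, showing either its raw contents or a distance gradient.
# It samples the RT texture parameter at mip 0 with point-sampled UVs based on TextureSize, treating red and green as stored coordinates.
# The offset length from each texel's UV, divided by Gradient Divisor, is blended with raw RGB by ShowDistance.
# A separate anti-aliased tiling grid node has no visible connection to that blend.
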Material /Volumetrics/Tools/JumpFlooding/Materials/Preview/M_PreviewRT
graph {
 properties {pixelDepthOffsetMode=PDOM_Legacy, shadingModel=MSM_Unlit, usedShadingModels=Unlit}
 nodes [
  n1 = MaterialFunctionCall!(
   UVs=TextureCoordinate()
   Texture Size=ComponentMask(None=VectorParameter() {parameterName=TextureSize, defaultValue=<rgba 256 256 0 0>}.RGB) {g=true, r=true}
  ) {materialFunction=/Engine/Functions/Engine_MaterialFunctions01/Texturing/PointSampledUVs}
  n2 = TextureSampleParameter2D!(UVs=@n1.Result, Level, Apply View MipBias) {constMipValue=0, mipValueMode=TMVM_MipLevel, parameterName=RT, texture=&Black}
  n3 = Subtract(A=@n1.Result, B=ComponentMask(None=@n2.RGB) {g=true, r=true})
  LinearInterpolate(
   A=@n2.RGB
   B=Divide(
    A=MaterialFunctionCall(Vector 2=@n3) {materialFunction=/Engine/Functions/Engine_MaterialFunctions02/Utility/VectorLength}.V2 Length
    B=ScalarParameter() {defaultValue=0.5, parameterName=Gradient Divisor}
   )
   Alpha=ScalarParameter() {parameterName=ShowDistance}
  )
  MaterialFunctionCall!(
   ScalarUV=Multiply(A=TextureCoordinate(), B=ScalarParameter() {defaultValue=256, parameterName=ss})
   Split=ScalarParameter() {defaultValue=0.01}
  ) {materialFunction=/Engine/EditorMaterials/TilingAAGrid}
 ]
}
