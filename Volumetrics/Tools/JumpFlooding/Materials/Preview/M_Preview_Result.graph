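# Unlit preview material for the Jump Flooding tool in Volumetrics, sampling its render targets.
# UVs snap to texel positions using the TextureSize parameter, with a sub-texel offset, to read mip 0 of RT_JumpFlood_01 and 02.
# The Index parameter blends the red-green channels of those two buffers.
# A separate node samples RT_JumpFlood_03_Result without the snapped UVs.
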Material /Volumetrics/Tools/JumpFlooding/Materials/Preview/M_Preview_Result
graph {
 properties {pixelDepthOffsetMode=PDOM_Legacy, shadingModel=MSM_Unlit, usedShadingModels=Unlit}
 nodes [
  n1 = ComponentMask(None=VectorParameter() {parameterName=TextureSize, defaultValue=<rgba 64 64 0 0>}.RGB) {g=true, r=true}
  n2 = Multiply(A=TextureCoordinate(), B=@n1)
  n3 = Add(A=Divide(A=Floor(None=@n2), B=@n1), B=Divide(A=Divide(B=@n1)))
  LinearInterpolate(
   A=ComponentMask(None=TextureSample!(UVs=@n3, Tex, Level, Apply View MipBias) {constMipValue=0, mipValueMode=TMVM_MipLevel, texture=/Volumetrics/Tools/JumpFlooding/Textures/RT_JumpFlood_02}.RGB) {g=true, r=true}
   B=ComponentMask(None=TextureSample(UVs=@n3) {constMipValue=0, mipValueMode=TMVM_MipLevel, texture=/Volumetrics/Tools/JumpFlooding/Textures/RT_JumpFlood_01}.RGB) {g=true, r=true}
   Alpha=ScalarParameter() {parameterName=Index}
  )
  TextureSample!(UVs, Tex, Apply View MipBias) {texture=/Volumetrics/Tools/JumpFlooding/Textures/RT_JumpFlood_03_Result}
 ]
}
